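; Parent Blueprint for modular planter props that scatters instanced foliage on the planter mesh's sockets.
; The construction script adds instances at socket transforms using the last static mesh component's mesh, jittering position, yaw and scale.
; Randomize swaps in a random or index-selected mesh from a list, re-jitters instances and writes palette colors as HSV packed into per-instance custom data.
; Reconstruct clears HISM instances, then calls the parent and Randomize.
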(event EventBeginPlay
  (|CallParentFunction) ; This node is disabled and will not be called.\nDrag off pins to build functionality.
)

(event Collision|EventActorBeginOverlap
  (params (OtherActor :type Actor%20Object%20Reference))
  (|CallParentFunction OtherActor) ; This node is disabled and will not be called.\nDrag off pins to build functionality.
)

(event EventTick
  (params (DeltaSeconds :type Float%20%28single-precision%29))
  (|CallParentFunction DeltaSeconds) ; This node is disabled and will not be called.\nDrag off pins to build functionality.
)

(fn GetAssignedMeshFromParent
  (Variables|Default|SetInstanced%20Mesh%20to%20Spawn (Variables|/Script/Engine.StaticMeshComponent|GetStaticMesh (Utilities|Array|Get(acopy) (Actor|GetComponentsByClass "/Script/Engine.StaticMeshComponent") (- (Utilities|Array|Length (Actor|GetComponentsByClass "/Script/Engine.StaticMeshComponent")) 1)))))

(fn Randomize
  (bind _returnvalue (Actor|GetComponentbyClass "/Script/Engine.HierarchicalInstancedStaticMeshComponent"))
  (bind _returnvalue_2 (Components|InstancedStaticMesh|GetInstanceCount _returnvalue))
  (bind _returnvalue_3 (- _returnvalue_2 -1))
  (bind _meshscale (Variables|RandomMesh|GetMeshScale))
  (bind _selectedfoliageindex (Variables|Default|GetSelectedFoliageIndex))
  (bind _output (CallFunction|SetMeshtoSpawn))
  (bind _output_get (Variables|Default|SetInstanced%20Mesh%20to%20Spawn _output))
  (if (Utilities|Array|IsValidIndex _meshscale _selectedfoliageindex)
    (Variables|AddonDeco|SetInstancedScale (Utilities|Array|Get(acopy) _meshscale _selectedfoliageindex))
    (else
      (Variables|AddonDeco|SetInstancedScale 1.0)))
  (bind _output_get_2 (Variables|Default|SetHISM_Component _returnvalue))
  (Variables|/Script/Engine.StaticMeshComponent|SetbEvaluateWorldPositionOffset true _returnvalue)
  (Components|StaticMesh|SetStaticMesh _output_get_2 _output_get)
  (for _index (range (+ _returnvalue_3 1))
    (Components|InstancedStaticMesh|UpdateInstanceTransform _returnvalue _index (Math|Transform|MakeTransform (Math|Vector|MakeVector (+ (.x (.location (Components|InstancedStaticMesh|GetInstanceTransform _returnvalue _index false))) (Math|Random|RandomFloatinRange (* (Variables|Default|GetInstanceOffset) -1.0) (Variables|Default|GetInstanceOffset))) (+ (.y (.location (Components|InstancedStaticMesh|GetInstanceTransform _returnvalue _index false))) (Math|Random|RandomFloatinRange (* (Variables|Default|GetInstanceOffset) -1.0) (Variables|Default|GetInstanceOffset))) (.z (.location (Components|InstancedStaticMesh|GetInstanceTransform _returnvalue _index false)))) (Math|Rotator|MakeRotator (.roll (.rotation (Components|InstancedStaticMesh|GetInstanceTransform _returnvalue _index false))) (.pitch (.rotation (Components|InstancedStaticMesh|GetInstanceTransform _returnvalue _index false))) (.yaw (Math|Random|RandomRotator true))) (Math|Vector|MakeVector (Math|Random|RandomFloatinRange (* (Variables|AddonDeco|GetInstancedScale) 0.9) (* (Variables|AddonDeco|GetInstancedScale) 1.1)) (Math|Random|RandomFloatinRange (* (Variables|AddonDeco|GetInstancedScale) 0.9) (* (Variables|AddonDeco|GetInstancedScale) 1.1)) (Math|Random|RandomFloatinRange (* (Variables|AddonDeco|GetInstancedScale) 0.9) (* (Variables|AddonDeco|GetInstancedScale) 1.1)))) false false false)
    (CallFunction|SetColorPerInstance _returnvalue _returnvalue_3)
    (Development|PrintString (Utilities|String|ToString(Integer) (Components|InstancedStaticMesh|GetInstanceCount (Variables|Default|GetHISM_Component))))))

(fn Reconstruct
  (for _array_element (Actor|GetComponentsByClass "/Script/Engine.HierarchicalInstancedStaticMeshComponent")
    (Components|InstancedStaticMesh|ClearInstances _array_element))
  (|CallParentFunction)
  (AddonDeco|Randomize))

(fn ResetTransforms
  (bind _output_get (Variables|Default|SetHISM_Component (Actor|GetComponentbyClass "/Script/Engine.HierarchicalInstancedStaticMeshComponent")))
  (for _index (range (+ (- (Components|InstancedStaticMesh|GetInstanceCount _output_get) -1) 1))
    (Components|InstancedStaticMesh|UpdateInstanceTransform (Variables|Default|GetHISM_Component) _index (Math|Transform|MakeTransform (.location (Components|InstancedStaticMesh|GetInstanceTransform (Variables|Default|GetHISM_Component) _index false)) (.rotation (Components|InstancedStaticMesh|GetInstanceTransform (Variables|Default|GetHISM_Component) _index false)) (.scale (Components|InstancedStaticMesh|GetInstanceTransform (Variables|Default|GetHISM_Component) _index false))) false false false)))

(fn SetColorPerInstance
  (params (InstancedComponent :type Instanced%20Static%20Mesh%20Component%20Object%20Reference) (LastInstanceIndex :type Integer))
  (bind (_r _g _b) (Utilities|Struct|BreakLinearColor (Math|Color|RGBtoHSV(LinearColor) (Utilities|Array|Get(acopy) (Variables|Default|GetInstanceColors) (Math|Random|RandomIntegerinRange 0 (- (Utilities|Array|Length (Variables|Default|GetInstanceColors)) 1))))))
  (Components|InstancedStaticMesh|SetNumCustomDataFloats InstancedComponent 1)
  (for _index (range (+ LastInstanceIndex 1))
    (Components|InstancedStaticMesh|SetCustomDataValue InstancedComponent _index 30 (Math|Color|Fortnite|PackHSVtoFloat _r _g _b) true)))

(fn SetMeshToSpawn
  (bind _meshlist (Variables|RandomMesh|GetMeshList))
  (bind (_outitem _outindex) (Utilities|Array|RandomArrayItem _meshlist))
  (bind _meshlist_2 (Variables|RandomMesh|GetMeshList))
  (bind _returnvalue (Utilities|Array|Length _meshlist_2))
  (bind _returnvalue_2 (- _returnvalue 1))
  (if (Variables|AddonDeco|GetUseRandomFoliage)
    (Variables|Default|SetSelectedFoliageIndex _outindex)
    (return (Utilities|ResolveSoftReference _outitem))
    (elif (<= (Variables|AddonDeco|GetFoliageIndex) _returnvalue_2)
      (bind _output_get (Variables|Default|SetSelectedFoliageIndex (Variables|AddonDeco|GetFoliageIndex)))
      (return (Utilities|ResolveSoftReference (Utilities|Array|Get(acopy) (Variables|RandomMesh|GetMeshList) _output_get)))
      (else
        (bind _output_get_2 (Variables|AddonDeco|SetFoliageIndex _returnvalue_2))
        (bind _output_get_3 (Variables|Default|SetSelectedFoliageIndex _output_get_2))
        (return (Utilities|ResolveSoftReference (Utilities|Array|Get(acopy) (Variables|RandomMesh|GetMeshList) _output_get_3)))))))

(fn ConstructionScript
  (bind _meshindex (Variables|RandomMesh|GetMeshIndex))
  (bind _meshscale (Variables|RandomMesh|GetMeshScale))
  (|CallParentFunction)
  (for _array_element (Transformation|GetAllSocketNames (Variables|Debug|GetStaticMeshComponent))
    (Utilities|Array|Add (Variables|Default|GetSocketTransforms) (Transformation|GetSocketTransform (Variables|Debug|GetStaticMeshComponent) _array_element "RTS_Actor"))
    (if (Utilities|Array|IsValidIndex _meshscale _meshindex)
      (Variables|AddonDeco|SetInstancedScale (Utilities|Array|Get(acopy) _meshscale _meshindex))
      (else
        (Variables|AddonDeco|SetInstancedScale 1.0)))
    (CallFunction|GetAssignedMeshfromParent)
    (Components|StaticMesh|SetStaticMesh (Variables|Default|GetFoliage_ISM) (Variables|Default|GetInstanced%20Mesh%20to%20Spawn))
    (Components|InstancedStaticMesh|AddInstances (Variables|Default|GetFoliage_ISM) (Variables|Default|GetSocketTransforms) false false true)
    (for _index (range (+ (- (Components|InstancedStaticMesh|GetInstanceCount (Variables|Default|GetFoliage_ISM)) -1) 1))
      (Components|InstancedStaticMesh|UpdateInstanceTransform (Variables|Default|GetFoliage_ISM) _index (Math|Transform|MakeTransform (Math|Vector|MakeVector (+ (.x (.location (Components|InstancedStaticMesh|GetInstanceTransform (Variables|Default|GetFoliage_ISM) _index false))) (Math|Random|RandomFloatinRange (* (Variables|Default|GetInstanceOffset) -1.0) (Variables|Default|GetInstanceOffset))) (+ (.y (.location (Components|InstancedStaticMesh|GetInstanceTransform (Variables|Default|GetFoliage_ISM) _index false))) (Math|Random|RandomFloatinRange (* (Variables|Default|GetInstanceOffset) -1.0) (Variables|Default|GetInstanceOffset))) (.z (.location (Components|InstancedStaticMesh|GetInstanceTransform (Variables|Default|GetFoliage_ISM) _index false)))) (Math|Rotator|MakeRotator (.roll (.rotation (Components|InstancedStaticMesh|GetInstanceTransform (Variables|Default|GetFoliage_ISM) _index false))) (.pitch (.rotation (Components|InstancedStaticMesh|GetInstanceTransform (Variables|Default|GetFoliage_ISM) _index false))) (.yaw (Math|Random|RandomRotator true))) (Math|Vector|MakeVector (Math|Random|RandomFloatinRange (* (Variables|AddonDeco|GetInstancedScale) 0.8) (* (Variables|AddonDeco|GetInstancedScale) 1.2)) (Math|Random|RandomFloatinRange (* (Variables|AddonDeco|GetInstancedScale) 0.8) (* (Variables|AddonDeco|GetInstancedScale) 1.2)) (Math|Random|RandomFloatinRange (* (Variables|AddonDeco|GetInstancedScale) 0.8) (* (Variables|AddonDeco|GetInstancedScale) 1.2)))) false false false))))
